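# Translucent unlit ribbon material for the Outlander GroundWave ability FX, usable on beam trails, Niagara ribbons and sprites.
# A procedural UV mask from sines and a powered U falloff, scaled by particle and dynamic alpha, blends an index-of-refraction value from 1 toward the Refraction parameter.
# Hue-shifted particle color times dynamic intensity feeds an emissive adjustment function.
# UV-blended vectors build a procedural normal, and a static switch enables camera offset.
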Material /Game/Abilities/Player/Outlander/Actives/GroundWave/FX/M_Ribbons
graph {
 properties {
  bUsedWithBeamTrails=true
  bUsedWithNiagaraRibbons=true
  bUsedWithParticleSprites=true
  blendMode=BLEND_Translucent
  pixelDepthOffsetMode=PDOM_Legacy
  refractionMethod=RM_IndexOfRefraction
  shadingModel=MSM_Unlit
  translucencyPass=MTP_BeforeDOF
  usedShadingModels=Unlit
 }
 nodes [
  n1 = DynamicParameter() {defaultValue=<rgba 1 0.9 1 1>, paramNames [Alpha,Refraction,BP_Intensity,BP_HueShift]}
  n2 = ParticleColor()
  n3 = TextureCoordinate() {vTiling=0.5}
  n4 = ComponentMask(None=OneMinus(None=@n3)) {r=true}
  n5 = Multiply(A=@n4, B=@n4)
  n6 = Multiply(A=@n5, B=@n5)
  n7 = Sine(None=ComponentMask(None=@n3) {g=true})
  n8 = Multiply(A=Sine(None=Multiply(A=@n6) {constB=0.5}), B=@n7)
  LinearInterpolate(B=@n1.Refraction, Alpha=Multiply(A=Multiply(A=@n2.A, B=@n8), B=@n1.Alpha)) {constA=1, constB=0.98}
  n9 = DynamicParameter() {defaultValue=<rgba 1 0.9 1 1>, paramNames [Alpha,Refraction,BP_Intensity,BP_HueShift]}
  MaterialFunctionCall(Emissive=Multiply(
    A=@n9.BP_Intensity
    B=MaterialFunctionCall(Hue Shift Percentage=@n9.BP_HueShift, Texture=@n2.RGB) {materialFunction=&HueShift}.Result
   )) {materialFunction=&MF_FX_EmissiveAdjustments}
  n10 = TextureCoordinate()
  n11 = LinearInterpolate(
   A=Constant3Vector() {constant=<rgba 0 0 1 0>}
   B=LinearInterpolate(A=Constant3Vector() {constant=<rgba 1 0 0 0>}, B=Constant3Vector() {constant=<rgba 0 1 0 0>}, Alpha=ComponentMask(None=@n10) {g=true})
   Alpha=@n7
  )
  Normalize(VectorInput=ConstantBiasScale(None=Normalize(VectorInput=@n11)) {bias=-0.5, scale=2})
  StaticSwitchParameter(
   True=MaterialFunctionCall(Offset Amount=ScalarParameter() {parameterName=CamOffset}) {materialFunction=/Engine/Functions/Engine_MaterialFunctions02/WorldPositionOffset/CameraOffset}.Clamped Camera Offset
   False=Constant()
  ) {parameterName=CamOffset}
 ]
}
